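# Niagara system for the FileTransfer vehicle boost cosmetic, drawing sprites and trails at the car's filtered skeletal-mesh sockets.
# A burst emitter holds looping socket sprites, StarSprites001 streams fading SubUV sprites backward, and a quality-scaled ribbon emitter renders four trail layers.
# Painted-variant tables feed per-layer colors, with bIsPainted choosing them over fixed defaults.
# GameModeID adjusts ribbon lifetime and offset, while floored engine time computes stepped ribbon X offsets.
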NiagaraSystem /VehicleCosmetics/Boosts/Boost_FileTransfer/NS/NS_Boost_FileTransfer
dependencies {
 usedDataInterfaces [
  &NiagaraDataInterfaceSpriteRendererInfo
  &NiagaraDataInterfaceCurve
  &NiagaraDataInterfaceSkeletalMesh
 ]
 usedDynamicInputs []
 usedModules [
  &self:SystemSpawnScript.SystemScriptSource.SystemScriptGraph.NiagaraNodeAssignment_0.SetVariables_#1
  &SystemState
 ]
}
emitters {
 NE_VC_Boost_Sprite_Parent {
  emitterUpdateScript [
   &EmitterState(
    Inactive Response=::NewEnumerator1
    Life Cycle Mode=::NewEnumerator1
    Loop Behavior=::NewEnumerator1
    Loop Duration Mode=::NewEnumerator1
    UseLoopDelay=false
   )
   &VC_FilteredSocketCount()
   &SpawnBurst_Instantaneous(Spawn Count=4, Use Spawn Probability=false, UseLoopCountLimit=false)
   &SpawnRate(SpawnRate=50, Use Spawn Probability=false) {enabled=false}
  ]
  particleSpawnScript [
   &VC_PaintedVariantsOverride(
    Black=<rgba 0.03 0.03 0.03 1>
    BurntSienna=<rgba 0.286 0.0532 0.0174 1>
    Cobalt=<rgba 0.1962 0.5445 1 1>
    Crimson=<rgba 0.7101 0 0.0192 1>
    ForestGreen=<rgba 0 0.6719 0.0339 1>
    Grey=<rgba 0.3698 0.3698 0.3698 1>
    Lime=<rgba 0.4332 0.8663 0 1>
    Orange=<rgba 0.7934 0.5388 0 1>
    Pink=<rgba 1 0.3455 0.762 1>
    Purple=<rgba 0.1622 0.0443 0.3316 1>
    R1 Black:LinearColor=<rgba 0 0 0 1.4>
    R1 BurntSienna:LinearColor=<rgba 0.217 0.1042 0.0868 1.25>
    R1 Cobalt:LinearColor=<rgba 0.0632 0.2054 0.6319 1.25>
    R1 Crimson:LinearColor=<rgba 0.4844 0 0.0131 1.25>
    R1 ForestGreen:LinearColor=<rgba 0 0.3941 0.0199 1.25>
    R1 Grey:LinearColor=<rgba 0.3281 0.3281 0.3281 1.25>
    R1 Lime:LinearColor=<rgba 0.2511 0.408 0.0942 1.25>
    R1 Orange:LinearColor=<rgba 0.5434 0.2454 0.0859 1.25>
    R1 Pink:LinearColor=<rgba 0.3941 0.1088 0.2903 1.25>
    R1 Purple:LinearColor=<rgba 0.2287 0 0.5573 1.25>
    R1 Saffron:LinearColor=<rgba 0.4566 0.4524 0 1.25>
    R1 SkyBlue:LinearColor=<rgba 0 0.3338 0.5017 1.25>
    R1 TitaniumWhite:LinearColor=<rgba 0.5503 0.5503 0.5503 1>
    Saffron=<rgba 0.5319 0.5503 0 1>
    SkyBlue=<rgba 0.1476 0.7148 1 1>
    TitaniumWhite=<rgba 1 1 1 1>
    Use R1 Painted Overrides=true
   )
   &InitializeParticle(
    Color Mode=::NewEnumerator0
    Sprite Rotation Mode=::NewEnumerator0
    Sprite Size Mode=::NewEnumerator3
    Uniform Sprite Size=75
    UsePositionOffset=false
    Write Material Random=true
   )
   &PartitionParticles(Number Of Partitions=$Emitter.PortCount)
   &SkeletalMeshLocation(
    Bone / Socket Index=$Particles.PartitionParticles.Partition
    Bone / Socket Sampling Mode=::NewEnumerator5
    MaskSpawnGroup=false
    Mesh Sampling Type=::NewEnumerator6
    Sampled Position Offset=&AddVectorToPosition(Position=<-70 0 0>, Vector=$User.DTCarOffset)
    Skeletal Mesh=$System.SkeletalMesh
   )
   &AddVelocity(Velocity=<-1500 0 0>) {enabled=false}
   /SetVariables(
    Particles.BC_Edge:LinearColor=&MakeCustomLinearColorFromBool(Bool=$User.bIsPainted, True Color=$Particles.Color, False Color=<rgba 0.6 0.32 0 1>)
    Particles.BC_Fill:LinearColor=&MakeCustomLinearColorFromBool(Bool=$User.bIsPainted, True Color=$Particles.R1 Color, False Color=<rgba 0.4 0.2533 0 1>)
    Particles.SpriteSize_Fill:Vector2f=&Vector2DFromFloat(Value=76)
   ) {isSetParameters=true}
  ]
  particleUpdateScript [
   &ParticleState(Kill Particles When Lifetime Has Elapsed=false, Loop Particles Lifetime=true)
   &InheritSourceMovement(
    Applied Position Scale=&Multiply_VectorByFloat(
     Vector=<1 1 1>
     Float=&FloatFromCurve(
      FloatCurve=DataInterface() {propertyValues {Curve {keys [
          {interpMode=RCIM_Cubic, tangentMode=RCTM_Break, time=0, value=1}
          {arriveTangent=-2.3961, interpMode=RCIM_Cubic, tangentMode=RCTM_Break, time=1, value=0}
         ]}}}
      Scale Curve=1.5
     )
    )
    ApplyForce=false
    ApplyPosition=true
    ApplyPositionFalloff=false
   ) {enabled=false}
   &SolveForcesAndVelocity(Clamp Velocity=false, Limit Acceleration=false)
  ]
  rendererClasses [&NiagaraSpriteRendererProperties]
  renderers [
   {rendererClass=&NiagaraSpriteRendererProperties}
   {rendererClass=&NiagaraSpriteRendererProperties, rendererIndex=1}
   {rendererClass=&NiagaraSpriteRendererProperties, rendererIndex=2}
   {rendererClass=&NiagaraSpriteRendererProperties, rendererIndex=3}
  ]
 }
 StarSprites001 {
  emitterUpdateScript [
   &EmitterState(Inactive Response=::NewEnumerator1, Life Cycle Mode=::NewEnumerator1, Loop Duration=1, UseLoopDelay=false)
   &VC_FilteredSocketCount()
   &SpawnRate(SpawnRate=12, Use Spawn Probability=false)
  ]
  particleSpawnScript [
   &VC_PaintedVariantsOverride(
    Black=<rgba 0.05 0.05 0.05 1.5>
    BurntSienna=<rgba 0.1649 0.0798 0.0667 1>
    Cobalt=<rgba 0.0855 0.092 0.4358 1>
    Crimson=<rgba 0.2413 0 0.0065 1>
    ForestGreen=<rgba 0 0.2413 0.0122 1>
    Grey=<rgba 0.1233 0.1233 0.1233 1>
    Lime=<rgba 0.1507 0.2135 0.0879 1>
    Orange=<rgba 0.2691 0.1483 0.0836 1>
    Pink=<rgba 0.2517 0.1412 0.2115 1>
    Purple=<rgba 0.1728 0.084 0.3003 1>
    R1 Black:LinearColor=<rgba 0.3 0.3 0.3 1>
    R1 BurntSienna:LinearColor=<rgba 0.4913 0.1517 0.0606 1>
    R1 Cobalt:LinearColor=<rgba 0.2448 0.4336 1 1>
    R1 Crimson:LinearColor=<rgba 0.467 0.0997 0.1097 1>
    R1 ForestGreen:LinearColor=<rgba 0.1534 0.7552 0.1838 1>
    R1 Grey:LinearColor=<rgba 0.6788 0.6788 0.6788 1>
    R1 Lime:LinearColor=<rgba 0.5728 0.7865 0.3591 1>
    R1 Orange:LinearColor=<rgba 0.9219 0.4522 0 1>
    R1 Pink:LinearColor=<rgba 0.8385 0.4673 0.7035 1>
    R1 Purple:LinearColor=<rgba 0.4062 0.2159 0.6795 1>
    R1 Saffron:LinearColor=<rgba 0.9067 1 0.3073 1>
    R1 SkyBlue:LinearColor=<rgba 0 0.6654 1 1>
    R1 TitaniumWhite:LinearColor=<rgba 1 1 1 1>
    Saffron=<rgba 0.2612 0.2934 0.0698 1>
    SkyBlue=<rgba 0.0323 0.1714 0.2413 1>
    TitaniumWhite=<rgba 0.2726 0.2726 0.2726 1>
    Use R1 Painted Overrides=true
   )
   &InitializeParticle(
    Color Mode=::NewEnumerator0
    Lifetime=0.5
    Sprite Rotation Mode=::NewEnumerator2
    Sprite Size Mode=::NewEnumerator3
    Uniform Sprite Size=80
    UsePositionOffset=false
    Write Material Random=true
   )
   &PartitionParticles(Number Of Partitions=$Emitter.PortCount)
   &SkeletalMeshLocation(
    Bone / Socket Index=$Particles.PartitionParticles.Partition
    Bone / Socket Sampling Mode=::NewEnumerator5
    MaskSpawnGroup=false
    Mesh Sampling Type=::NewEnumerator6
    Sampled Position Offset=<-110 0 15>
    Skeletal Mesh=$System.SkeletalMesh
   )
   &AddVelocity(Velocity=<-1200 0 0>)
  ]
  particleUpdateScript [
   &ParticleState()
   &SolveForcesAndVelocity(Clamp Velocity=false, Limit Acceleration=false)
   &SubUVAnimation(
    "Sprite Renderer "=DataInterface() {propertyValues {SpriteRenderer=&self:StarSprites001_0.NiagaraSpriteRendererProperties_0}}
    UseEndFrame=false
    UseStartFrame=false
   )
   /SetVariables(
    Particles.BC_Edge:LinearColor=&MakeCustomLinearColorFromBool(
     Bool=$User.bIsPainted
     True Color=&MultiplyLinearColor(
      Linear Color=$Particles.R1 Color
      Scale Factor=&MakeVector4FromVectorAndFloat(
       "Vector (RGB)"=<1 1 1>
       "Float (Alpha)"=&FloatFromCurve(FloatCurve=DataInterface() {propertyValues {Curve {
           defaultValue=1
           keys [
            {arriveTangent=-3.3543, interpMode=RCIM_Cubic, leaveTangent=-3.3543, tangentMode=RCTM_User, time=0.35, value=1}
            {arriveTangent=-0.0722, interpMode=RCIM_Cubic, leaveTangent=-0.0722, tangentMode=RCTM_User, time=1, value=0}
           ]
          }}})
      )
     )
     False Color=&MultiplyLinearColor(
      Linear Color=<rgba 0.6 0.6 0.6 1>
      Scale Factor=&MakeVector4FromVectorAndFloat(
       "Vector (RGB)"=<1 1 1>
       "Float (Alpha)"=&FloatFromCurve(FloatCurve=DataInterface() {propertyValues {Curve {
           defaultValue=1
           keys [
            {arriveTangent=-3.3543, interpMode=RCIM_Cubic, leaveTangent=-3.3543, tangentMode=RCTM_User, time=0.35, value=1}
            {arriveTangent=-0.0722, interpMode=RCIM_Cubic, leaveTangent=-0.0722, tangentMode=RCTM_User, time=1, value=0}
           ]
          }}})
      )
     )
    )
    Particles.Fill Color:LinearColor=&MakeCustomLinearColorFromBool(
     Bool=$User.bIsPainted
     True Color=&MultiplyLinearColor(
      Linear Color=$Particles.Color
      Scale Factor=&MakeVector4FromVectorAndFloat(
       "Vector (RGB)"=<1 1 1>
       "Float (Alpha)"=&FloatFromCurve(FloatCurve=DataInterface() {propertyValues {Curve {
           defaultValue=1
           keys [
            {arriveTangent=-3.3839, interpMode=RCIM_Cubic, leaveTangent=-3.3839, tangentMode=RCTM_User, time=0.25, value=1}
            {arriveTangent=-0.0722, interpMode=RCIM_Cubic, leaveTangent=-0.0722, tangentMode=RCTM_User, time=1, value=0}
           ]
          }}})
      )
     )
     False Color=&MultiplyLinearColor(
      Linear Color=<rgba 0.25 0.25 0.25 1>
      Scale Factor=&MakeVector4FromVectorAndFloat(
       "Vector (RGB)"=<1 1 1>
       "Float (Alpha)"=&FloatFromCurve(FloatCurve=DataInterface() {propertyValues {Curve {
           defaultValue=1
           keys [
            {arriveTangent=-3.3839, interpMode=RCIM_Cubic, leaveTangent=-3.3839, tangentMode=RCTM_User, time=0.25, value=1}
            {arriveTangent=-0.0722, interpMode=RCIM_Cubic, leaveTangent=-0.0722, tangentMode=RCTM_User, time=1, value=0}
           ]
          }}})
      )
     )
    )
   ) {isSetParameters=true}
  ]
  rendererClasses [&NiagaraSpriteRendererProperties]
  renderers [
   {rendererClass=&NiagaraSpriteRendererProperties}
   {rendererClass=&NiagaraSpriteRendererProperties, rendererIndex=1}
   {rendererClass=&NiagaraSpriteRendererProperties, rendererIndex=2}
  ]
 }
 ribbon {
  emitterUpdateScript [
   &EmitterState(Inactive Response=::NewEnumerator1, Life Cycle Mode=::NewEnumerator1, Loop Duration=1, UseLoopDelay=false)
   &VC_FilteredSocketCount()
   &VC_RateQualityHandler(Epic Spawn Rate=45, Low Spawn Rate=30, Use Custom Rate?=true)
   &VC_SpawnRibbon_v1(Number of Ribbons=$Emitter.PortCount, Spawn Time=-1, SpawnRate=$Emitter.Spawn Rate)
   /SetVariables(Emitter.OffsetX=HlslExpression() {hlslExpression="floor((Engine.Time * 0.2) * 40 + 0.02) * 0.03"}) {isSetParameters=true}
   /SetVariables(Emitter.OffsetX Glow=HlslExpression() {hlslExpression="floor((Engine.Time * 0.2) * 40 + 0.02) * 0.03"}) {isSetParameters=true}
  ]
  particleSpawnScript [
   &VC_PaintedVariantsOverride(
    Black=<rgba 1 1 1 0.25>
    BurntSienna=<rgba 1.5 0.7495 0.4766 0.5>
    Cobalt=<rgba 0.668 1.001 2 0.4>
    Crimson=<rgba 1 0.3385 0.3565 0.4>
    ForestGreen=<rgba 0.2066 1 0.2467 0.5>
    Grey=<rgba 0.9 0.9 0.9 0.85>
    Lime=<rgba 0.8345 1.25 0 0.5>
    Orange=<rgba 3 1.5075 0.151 0.1>
    Pink=<rgba 1.8 1.0781 1.7516 0.4>
    Purple=<rgba 0.6903 0.3354 1.2 0.851>
    R1 Black:LinearColor=<rgba 1 1 1 0.5>
    R1 BurntSienna:LinearColor=<rgba 1 0.56 0.4 1.2>
    R1 Cobalt:LinearColor=<rgba 0.668 1.001 2 0.8>
    R1 Crimson:LinearColor=<rgba 1 0.5156 0.5288 1>
    R1 ForestGreen:LinearColor=<rgba 0.3 1 0.3119 1.25>
    R1 Grey:LinearColor=<rgba 0.852 0.852 0.852 1>
    R1 Lime:LinearColor=<rgba 0.8756 1.25 0.1237 0.65>
    R1 Orange:LinearColor=<rgba 1 0.576 0.349 1.5>
    R1 Pink:LinearColor=<rgba 0.8906 0.5334 0.8667 2>
    R1 Purple:LinearColor=<rgba 0.6469 0.401 1 3>
    R1 Saffron:LinearColor=<rgba 0.8274 0.9 0 2>
    R1 SkyBlue:LinearColor=<rgba 0.2995 1.0983 1.5 1.25>
    R1 TitaniumWhite:LinearColor=<rgba 1 1 1 1>
    R2 Black:LinearColor=<rgba 0.3 0.3 0.3 2>
    R2 BurntSienna:LinearColor=<rgba 0.5608 0.2284 0.1392 2>
    R2 Cobalt:LinearColor=<rgba 0.6641 0.873 1.5 1>
    R2 Crimson:LinearColor=<rgba 1 0 0.0271 1.5>
    R2 ForestGreen:LinearColor=<rgba 0 1 0.0505 1>
    R2 Grey:LinearColor=<rgba 0.6267 0.6267 0.6267 1>
    R2 Lime:LinearColor=<rgba 0.75 1.5 0 1>
    R2 Orange:LinearColor=<rgba 2 0.8174 0.184 1>
    R2 Pink:LinearColor=<rgba 1.5 1.1276 1.3646 1>
    R2 Purple:LinearColor=<rgba 1.1381 0.5382 2 1>
    R2 Saffron:LinearColor=<rgba 0.9472 1 0 1>
    R2 SkyBlue:LinearColor=<rgba 0.2734 1.0896 1.5 1>
    R2 TitaniumWhite:LinearColor=<rgba 1.25 1.25 1.25 1>
    R3 Black:LinearColor=<rgba 0.6059 0.6059 0.6059 0.25>
    R3 BurntSienna:LinearColor=<rgba 0.5677 0.1753 0.07 0>
    R3 Cobalt:LinearColor=<rgba 0.2428 0.3752 0.7726 1.5>
    R3 Crimson:LinearColor=<rgba 1 0.1267 0.1504 0.5>
    R3 ForestGreen:LinearColor=<rgba 0 0.8767 0.0443 1>
    R3 Grey:LinearColor=<rgba 0.2552 0.2552 0.2552 0>
    R3 Lime:LinearColor=<rgba 0.3446 0.6892 0 1>
    R3 Orange:LinearColor=<rgba 1.5 0.6911 0.2578 1>
    R3 Pink:LinearColor=<rgba 1 0.6997 0.8908 1>
    R3 Purple:LinearColor=<rgba 0.513 0 1.25 1>
    R3 Saffron:LinearColor=<rgba 1 0.8133 0.2 1>
    R3 SkyBlue:LinearColor=<rgba 0 0.8317 1.25 1>
    Saffron=<rgba 1.25 1.25 0 0.25>
    SkyBlue=<rgba 0.5613 1.045 1.2882 1>
    TitaniumWhite=<rgba 1 1 1 0.5>
    Use R1 Painted Overrides=true
    Use R2 Painted Overrides=true
    Use R3 Painted Overrides=true
   )
   &VC_InitializeRibbon_v1(
    Color=$Particles.Color
    Custom Offset=true
    Enable Skeletal Position=true
    Lifetime=&MergeHelper_FloatSwitch(0=1.3, 1=1.3, 2=0.6, 3=?, Index=$User.GameModeID, Num=::NewEnumerator1)
    Local Vector Offset=&AddVectorToPosition(
     Position=&MergeHelper_VectorSwitch(0=<-30 0 0>, 1=<-30 0 0>, 2=<-55 0 0>, 3=?, Index=$User.GameModeID, Num=::NewEnumerator1)
     Vector=$User.DTCarOffset
    )
    Offset Amount=1
    Ribbon Width=40
    Write Material Random=false
   )
   &AddVelocity(Velocity=<-1350 0 0>) {module=AddVelocity001}
  ]
  particleUpdateScript [
   &ParticleState()
   &InheritSourceMovement(
    Applied Position Scale=&VectorFromFloat(Value=&FloatFromCurve(
      FloatCurve=DataInterface() {propertyValues {Curve {
         defaultValue=1
         keys [<at 0 1 RCIM_Cubic>, {arriveTangent=-3.2426, interpMode=RCIM_Cubic, leaveTangent=-3.2426, tangentMode=RCTM_User, time=1, value=0}]
        }}}
      Scale Curve=1.1
     ))
    ApplyForce=false
    ApplyPosition=true
    ApplyPositionFalloff=false
   )
   &SolveForcesAndVelocity(Clamp Velocity=false, Limit Acceleration=false)
   &VC_UpdateRibbon_v1(Enable Skeletal Position=true)
   /SetVariables(Particles.CUBE GEN=400, Particles.Scale Glow=300, Particles.Scale Principal=50, Particles.Scale Secondary=60) {isSetParameters=true}
   /SetVariables(
    Particles.BC_Cube:LinearColor=&MakeCustomLinearColorFromBool(
     Bool=$User.bIsPainted
     True Color=&MultiplyLinearColor(
      Linear Color=$Particles.R2 Color
      Scale Factor=&MakeVector4FromVectorAndFloat(
       "Vector (RGB)"=<1 1 1>
       "Float (Alpha)"=&FloatFromCurve(FloatCurve=DataInterface() {propertyValues {Curve {keys [<at 0.05 -0.5>,<at 0.2 1>,<at 0.9 0>]}}})
      )
     )
     False Color=&MultiplyLinearColor(
      Linear Color=<rgba 0.1 1 0.7 1>
      Scale Factor=&MakeVector4FromVectorAndFloat(
       "Vector (RGB)"=<1 1 1>
       "Float (Alpha)"=&FloatFromCurve(FloatCurve=DataInterface() {propertyValues {Curve {keys [<at 0.05 -0.5>,<at 0.2 1>,<at 0.9 0>]}}})
      )
     )
    )
    Particles.BC_Glow:LinearColor=&MakeCustomLinearColorFromBool(
     Bool=$User.bIsPainted
     True Color=&MultiplyLinearColor(
      Linear Color=$Particles.R3 Color
      Scale Factor=&MakeVector4FromVectorAndFloat(
       "Vector (RGB)"=<1 1 1>
       "Float (Alpha)"=&FloatFromCurve(FloatCurve=DataInterface() {propertyValues {Curve {keys [
            <at 0.05 -0.5>
            {arriveTangent=-0.4328, interpMode=RCIM_Cubic, leaveTangent=-0.4328, tangentMode=RCTM_User, time=0.2, value=1}
            {arriveTangent=-2.0719, interpMode=RCIM_Cubic, leaveTangent=-2.0719, tangentMode=RCTM_User, time=0.9, value=0}
           ]}}})
      )
     )
     False Color=&MultiplyLinearColor(
      Linear Color=<rgba 0.1 1 0.7 1>
      Scale Factor=&MakeVector4FromVectorAndFloat(
       "Vector (RGB)"=<1 1 1>
       "Float (Alpha)"=&FloatFromCurve(FloatCurve=DataInterface() {propertyValues {Curve {keys [
            <at 0.05 -0.5>
            {arriveTangent=-0.4328, interpMode=RCIM_Cubic, leaveTangent=-0.4328, tangentMode=RCTM_User, time=0.2, value=1}
            {arriveTangent=-2.0719, interpMode=RCIM_Cubic, leaveTangent=-2.0719, tangentMode=RCTM_User, time=0.9, value=0}
           ]}}})
      )
     )
    )
    Particles.BC_Principal:LinearColor=&MakeCustomLinearColorFromBool(
     Bool=$User.bIsPainted
     True Color=&MultiplyLinearColor(
      Linear Color=$Particles.Color
      Scale Factor=&MakeVector4FromVectorAndFloat("Vector (RGB)"=<1 1 1>, "Float (Alpha)"=1)
     )
     False Color=&MultiplyLinearColor(
      Linear Color=<rgba 0.4 1 0.85 1>
      Scale Factor=&MakeVector4FromVectorAndFloat("Vector (RGB)"=<1 1 1>, "Float (Alpha)"=1)
     )
    )
    Particles.BC_Secondary:LinearColor=&MakeCustomLinearColorFromBool(
     Bool=$User.bIsPainted
     True Color=&MultiplyLinearColor(
      Linear Color=$Particles.R1 Color
      Scale Factor=&MakeVector4FromVectorAndFloat(
       "Vector (RGB)"=<1 1 1>
       "Float (Alpha)"=&FloatFromCurve(FloatCurve=DataInterface() {propertyValues {Curve {defaultValue=1, keys [{tangentMode=RCTM_User, time=0, value=1}, {tangentMode=RCTM_User, time=0.7, value=0}]}}})
      )
     )
     False Color=&MultiplyLinearColor(
      Linear Color=<rgba 0.4 1 0.85 1>
      Scale Factor=&MakeVector4FromVectorAndFloat(
       "Vector (RGB)"=<1 1 1>
       "Float (Alpha)"=&FloatFromCurve(FloatCurve=DataInterface() {propertyValues {Curve {defaultValue=1, keys [{tangentMode=RCTM_User, time=0, value=1}, {tangentMode=RCTM_User, time=0.7, value=0}]}}})
      )
     )
    )
   ) {isSetParameters=true}
  ]
  rendererClasses [&NiagaraRibbonRendererProperties]
  renderers [
   {rendererClass=&NiagaraRibbonRendererProperties}
   {rendererClass=&NiagaraRibbonRendererProperties, rendererIndex=1}
   {rendererClass=&NiagaraRibbonRendererProperties, rendererIndex=2}
   {rendererClass=&NiagaraRibbonRendererProperties, rendererIndex=3}
  ]
 }
}
summary {
 systemName=NS_Boost_FileTransfer
 emitters [
  {emitterName=StarSprites001, rendererClasses [&NiagaraSpriteRendererProperties]}
  {emitterName=NE_VC_Boost_Sprite_Parent, rendererClasses [&NiagaraSpriteRendererProperties]}
  {emitterName=ribbon, rendererClasses [&NiagaraRibbonRendererProperties]}
 ]
 userVariables [
  User.bIsFrontendPreview {type=NiagaraBool}
  User.bIsPainted {type=NiagaraBool}
  User.ColorVariantID {type=NiagaraInt32, defaultValue=2}
  User.DTCarOffset {type=Vector3f, description="", defaultValue {struct=Vector3f, value=<0 0 0>}}
  User.GameModeID {type=NiagaraInt32}
  User.PaintedColor {type=LinearColor, description="", defaultValue {struct=LinearColor, value=<rgba 1 1 1 1>}}
  User.PaintedScalar {type=NiagaraFloat, defaultValue=1, description=""}
  User.PortCount {type=NiagaraInt32, defaultValue=2}
  User.Position_Sampled Position Offset {type=NiagaraPosition, description="", defaultValue {}}
  User.RibbonHLScalar {type=NiagaraFloat, defaultValue=4, description=""}
  User.RibbonInsideColor {type=NiagaraFloat, defaultValue=1, description=""}
  User.RibbonPaintedScalar {type=NiagaraFloat, defaultValue=1, description=""}
  User.SingleBoostPos {type=Vector3f, description="", defaultValue {struct=Vector3f, value=<0 0 0>}}
  User.StarOutlineScalar {type=NiagaraFloat, defaultValue=2, description=""}
  User.StarsSpritesScalar {type=NiagaraFloat, defaultValue=1, description=""}
 ]
}
